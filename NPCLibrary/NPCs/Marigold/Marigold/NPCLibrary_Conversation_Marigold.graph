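# Conversation database for the NPC Marigold, supplying the greeting and menu branches when a player talks to the character.
# The entry point leads to one speech node with a default line and contextual lines gated by time since last match, first conversation, Bars held or specific outfits.
# Each contextual line takes its priority from a curve row.
# The speech node then links to shared Basic Services, NPC Library Quests and Season 17 Week 2 conversations.
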ConversationDatabase /NPCLibrary/NPCs/Marigold/Marigold/NPCLibrary_Conversation_Marigold
entry_points [{tag=Conversation.Entry.Marigold, destinations [#1]}]
nodes [
 ConversationEntryPointNode_7 {
  id=#1
  class=ConversationEntryPointNode
  outputs [#2]
  properties {entryTag {tagName=Conversation.Entry.Marigold}}
 }
 Speech_C_4 {
  id=#2
  class=Speech_C
  outputs [#3, #4, #5]
  properties {
   nodeName=""
   generalConfig {
    defaultMessage=One step closer to the answers.
    contextualMessages [
     {
      message=I hope you've been practicing your handbrake turns.
      contextRequirements [{requirement {class=FortControllerRequirement_CheckTimeSinceLastPlayedMatch}}]
      priority {curve {rowName=Tandem.Speech.Priority.WhileSincePlayed}}
     }
     {
      message=I'm looking for a getaway driver. Think you could be the one?
      contextRequirements [{requirement {class=FortControllerRequirement_IsFirstEverConversationWithNPC}}]
      priority {curve {rowName=Tandem.Speech.Priority.FirstEverInteraction}}
     }
     {
      message="Put all those Bars in the trunk and let's <b>drive</>."
      contextRequirements [{requirement {class=FortControllerRequirement_HasBars, properties {thresholdTestConfig {requirement=GreaterThanOrEqual}}}}]
      priority {curve {rowName=Tandem.Speech.Priority.HighTotalBars}}
     }
     {
      message="It's happening <b>soon</>."
      requirementMatchPolicy=RequireAny
      contextRequirements [{requirement {
         class=FortControllerRequirement_HasCID
         properties {allowedCIds [
           /BRCosmetics/Athena/Items/Cosmetics/Characters/CID_694_Athena_Commando_M_CatBurglar
           /BRCosmetics/Athena/Items/Cosmetics/Characters/CID_923_Athena_Commando_M_ParcelGold
          ]}
        }}]
      priority {curve {rowName=Tandem.Speech.Priority.OutfitReaction}}
     }
     {
      message="Stick to the <b>plan</>. "
      requirementMatchPolicy=RequireAny
      contextRequirements [{requirement {
         class=FortControllerRequirement_HasCID
         properties {allowedCIds [/BRCosmetics/Athena/Items/Cosmetics/Characters/CID_770_Athena_Commando_F_MechanicalEngineer]}
        }}]
      priority {curve {rowName=Tandem.Speech.Priority.OutfitReaction}}
     }
    ]
   }
  }
 }
 ConversationLinkNode_0 {id=#5, class=ConversationLinkNode, properties {remoteEntryTag {tagName=Conversation.Entry.BasicServices}}}
 ConversationLinkNode_1 {id=#4, class=ConversationLinkNode, properties {remoteEntryTag {tagName=Conversation.Entry.NPCLibrary.Quests}}}
 ConversationLinkNode_2 {id=#3, class=ConversationLinkNode, properties {remoteEntryTag {tagName=Conversation.Entry.Season17.Week2}}}
]
